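; BP_Test_MeshNetwork_Actor is a prototype test actor Blueprint in the MeshNetwork folder that shifts the game's time of day.
; Its custom event AddTOD adds the given offset to the current time of day, wraps the sum modulo 24, and sets the result through FortKismetLibrary.
; The construction script is empty, and no mesh network logic appears in the graph.
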(event Custom|AddTOD
  (params (TOD :type Float%20%28double-precision%29))
  (TimeOfDay|SetTimeOfDay (Math|Float|%25%28Float%29 (+ TOD (TimeOfDay|GetTimeOfDay :!class /Script/FortniteGame.FortKismetLibrary)) 24.0) :!class /Script/FortniteGame.FortKismetLibrary))

(fn ConstructionScript)
